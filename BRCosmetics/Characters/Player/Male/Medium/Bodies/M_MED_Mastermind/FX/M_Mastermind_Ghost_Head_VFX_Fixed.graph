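# Gold metallic, animated goo material for the Mastermind character's ghost head VFX, flagged for skeletal meshes and Niagara mesh particles.
# MF_TwistedGoo supplies base attributes, with Base Color, Metallic and Roughness overridden for gold.
# A panning, actor-relative world-aligned noise texture drives roughness and normal-direction vertex displacement.
# That displacement plus a cosine sway along a local vector, masked by inverted vertex red, adds to the goo's World Position Offset.
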Material /BRCosmetics/Characters/Player/Male/Medium/Bodies/M_MED_Mastermind/FX/M_Mastermind_Ghost_Head_VFX_Fixed
graph {
 properties {
  bUseMaterialAttributes=true
  bUsedWithMeshDeformer=true
  bUsedWithNiagaraMeshParticles=true
  bUsedWithSkeletalMesh=true
  pixelDepthOffsetMode=PDOM_Legacy
  usedShadingModels=Default Lit
 }
 nodes [
  n1 = GetMaterialAttributes(None=MaterialFunctionCall!(In, Mask, isVehicle, isMech) {materialFunction=/Game/Characters/Player/Male/Medium/Bodies/M_MED_Mastermind/Materials/MF_TwistedGoo}.Result) {attributeGetTypes [#1]}
  n2 = VectorParameter() {group=Goo, parameterName=Panning Speed, defaultValue=<rgba 0 0.1 0 0>}
  n3 = MaterialFunctionCall!(
   TextureObject=TextureObject() {texture=&2ndInfectedMapLinear}
   TextureSize=VectorParameter() {group=Goo, parameterName=Texture Size, defaultValue=<rgba 100 100 100 0>}.RGB
   WorldPosition=Subtract(A=ActorPositionWS(), B=WorldPosition().XYZ)
   Export Float 4
   In=Multiply(A=ComponentMask(None=@n2.RGB) {g=true, r=true}, B=ScalarParameter() {defaultValue=1, group=Goo, parameterName=Panner Speed Scale})
   World Space Normal
   ProjectionTransitionContrast
  ) {materialFunction=/Game/Athena/SafeZone/WorldAlignedTexture_ScalePan}
  n4 = Multiply(
   A=VertexNormalWS()
   B=LinearInterpolate(
    A=ScalarParameter() {defaultValue=-0.1, group=Goo, parameterName=WPO_Offset A}
    B=ScalarParameter() {defaultValue=1.5, group=Goo, parameterName=WPO_Offset B}
    Alpha=Desaturation(None=@n3.XYZ Texture)
   )
  )
  n5 = Time()
  n6 = LinearInterpolate(
   A=ScalarParameter() {defaultValue=-0.1, group=Goo, parameterName=Motion Min}
   B=ScalarParameter() {defaultValue=0.3, group=Goo, parameterName=Motion Max}
   Alpha=ConstantBiasScale(None=Cosine(None=@n5))
  )
  n7 = Transform(None=VectorParameter() {group=Goo, parameterName=Motion Vector, defaultValue=<rgba 1 0 1 0>}.RGB) {transformSourceType=TRANSFORMSOURCE_Local}
  n8 = Multiply(A=OneMinus(None=VertexColor().R), B=Add(A=Multiply(A=@n4, B=@n3.XYZ Texture), B=Multiply(A=@n6, B=@n7)))
  SetMaterialAttributes!(
   MaterialAttributes=SetMaterialAttributes!(MaterialAttributes=@n1.MaterialAttributes, World Position Offset=Add(A=@n1.World Position Offset, B=@n8)) {attributeSetTypes [#1]}
   Base Color=VectorParameter() {parameterName=GoldColor, defaultValue=<rgba 0.7656 0.5265 0.1769 1>}.RGB
   Metallic=ScalarParameter() {defaultValue=1, parameterName=Metallic}
   Roughness=LinearInterpolate(
    A=ScalarParameter() {defaultValue=0.1, parameterName=Gold Rougness Min}
    B=ScalarParameter() {defaultValue=0.4, parameterName=Gold Rougness Max}
    Alpha=ComponentMask(None=@n3.XYZ Texture) {g=true}
   )
  ) {attributeSetTypes [#2, #3, #4]}
 ]
}
