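# Niagara effect for the Mastermind cosmetic body's head, a single Mastermind_Goop emitter of mesh particles spawned on a user-supplied skeletal mesh.
# Two randomized-rate spawn groups differ in lifetime, scale, rotation and motion through spawn-group masks, with one also getting a small sphere offset.
# Particles swell, shrink and fade over life, with curve-ramped upward acceleration, heavy drag and randomized dynamic material parameters.
# The AddVelocity module is disabled, and User.Velocity is unreferenced.
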NiagaraSystem /BRCosmetics/Characters/Player/Male/Medium/Bodies/M_MED_Mastermind/FX/NS_Mastermind_Head_Ghost
dependencies {
 usedDataInterfaces [&NiagaraDataInterfaceMeshRendererInfo, &NiagaraDataInterfaceCurve]
 usedDynamicInputs []
 usedModules [&SystemState]
}
emitters {Mastermind_Goop {
  emitterUpdateScript [
   &EmitterState()
   &SpawnRate(SpawnRate=&RandomRangeFloat(Minimum=3, Maximum=6), Use Spawn Probability=false)
   &SpawnRate(SpawnGroup=1, SpawnRate=&RandomRangeFloat(Minimum=1, Maximum=6), Use Spawn Probability=false) {module=SpawnRate001}
  ]
  particleSpawnScript [
   &InitializeParticle(
    Lifetime Min:NiagaraFloat=&MaskFloatBySpawnGroup(Masked Float=1, Passthrough Float=1.5, Spawn Group Mask=1)
    Lifetime Mode=::NewEnumerator1
    Mesh Renderer Array Visibility Mode=::NewEnumerator3
    Mesh Renderer Info=DataInterface() {propertyValues {MeshRenderer=&self:Mastermind_Goop_0.NiagaraMeshRendererProperties_0}}
    Mesh Scale Mode=::NewEnumerator1
    Mesh Uniform Scale Max:NiagaraFloat=&MaskFloatBySpawnGroup(Masked Float=5, Passthrough Float=1, Spawn Group Mask=1)
    Mesh Uniform Scale Min:NiagaraFloat=&MaskFloatBySpawnGroup(Masked Float=1, Passthrough Float=0.75, Spawn Group Mask=1)
    UsePositionOffset=false
    Write Material Random=true
   )
   &InitialMeshOrientation(Use Rotation Vector=false)
   &DynamicMaterialParameters(
    Index 0 Param 1=0.1
    Index 0 Param 3=&RandomRangeFloat(Minimum=0.5, Maximum=1)
    Index 0 Param 4=&RandomRangeFloat(Minimum=50, Maximum=55)
    Module.Param0WriteEnabled=true
    Module.Param1WriteEnabled=false
    Module.Param2WriteEnabled=true
    Module.Param3WriteEnabled=true
   )
   &SkeletalMeshLocation(
    MaskSpawnGroup=false
    Mesh Sampling Type=::NewEnumerator5
    Orientation Sampling=::NewEnumerator1
    OverrideRandomSeed=false
    Skeletal Mesh=$User.SkeletalMesh
   )
   &ShapeLocation(
    MaskSpawnGroup=true
    Override Seed=false
    Spawn Group Mask=1
    Sphere Radius=&RandomRangeFloat(Minimum=1, Maximum=2)
    UseOwnerScale=true
   )
   &AddVelocity(Velocity=&MaskVectorBySpawnGroup(
     Masked Vector=<0 0 0>
     Passthrough Vector=&Add_Vector(A=&Multiply_VectorByFloat(Vector=$Particles.ShapeLocation.ShapeNormal, Float=1), B=<0 -10 0>)
    )) {enabled=false}
  ]
  particleUpdateScript [
   &ParticleState()
   &ScaleMeshSize(Scale Factor=&VectorFromFloat(Value=&FloatFromCurve(FloatCurve=DataInterface() {propertyValues {Curve {keys [
          {arriveTangent=5.8044, interpMode=RCIM_Cubic, leaveTangent=5.8044, tangentMode=RCTM_User, time=0, value=0}
          {arriveTangent=1.1084, interpMode=RCIM_Cubic, leaveTangent=1.1084, tangentMode=RCTM_User, time=0.25, value=1.125}
          {arriveTangent=-2.8011, interpMode=RCIM_Cubic, leaveTangent=-2.8011, tangentMode=RCTM_User, time=0.7835, value=0.3}
          <at 1 0 RCIM_Cubic>
         ]}}})))
   &UpdateMeshOrientation(
    Rotation Rate=&RandomRangeFloat(Minimum=-0.2, Maximum=0.2)
    Rotation Vector=&MaskVectorBySpawnGroup(Masked Vector=<1 1 1>, Passthrough Vector=<0 0 0>, Spawn Group Mask=1)
   )
   &ScaleColor(
    Scale Alpha=&FloatFromCurve(FloatCurve=DataInterface() {propertyValues {Curve {keys [
         {arriveTangent=7.9965, interpMode=RCIM_Cubic, leaveTangent=7.9965, leaveTangentWeight=0.0667, tangentMode=RCTM_User, time=0, value=0}
         {arriveTangentWeight=0.1067, interpMode=RCIM_Cubic, leaveTangentWeight=0.3202, tangentMode=RCTM_User, time=0.2, value=1}
         {arriveTangentWeight=0.3202, interpMode=RCIM_Cubic, leaveTangentWeight=0.1067, tangentMode=RCTM_User, time=0.8, value=1}
         {arriveTangent=-8.7139, arriveTangentWeight=0.0667, interpMode=RCIM_Cubic, leaveTangent=-8.7139, tangentMode=RCTM_User, time=1, value=0}
        ]}}})
    ScaleA=true
    ScaleRGB=true
   )
   &AccelerationForce(Acceleration=&MaskVectorBySpawnGroup(
     Masked Vector=<0 0 0>
     Passthrough Vector=&Multiply_VectorByFloat(
      Vector=&RandomRangeVector(Minimum=<0 0 12>, Maximum=<0 0 24>)
      Float=&FloatFromCurve(FloatCurve=DataInterface() {propertyValues {Curve {keys [<at 0 0 RCIM_Cubic>, {arriveTangent=27.1737, interpMode=RCIM_Cubic, leaveTangent=27.1737, tangentMode=RCTM_User, time=0.6, value=10}]}}})
     )
    ))
   &Drag(Drag=&RandomRangeFloat(Minimum=15, Maximum=25), Use Linear Drag=true, Use Rotational Drag=true)
   &SolveForcesAndVelocity(Clamp Velocity=false, Limit Acceleration=false)
  ]
  rendererClasses [&NiagaraMeshRendererProperties]
  renderers [{rendererClass=&NiagaraMeshRendererProperties, rendererIndex=0}]
 }}
summary {
 systemName=NS_Mastermind_Head_Ghost
 emitters [{emitterName=Mastermind_Goop, rendererClasses [&NiagaraMeshRendererProperties]}]
 userVariables [
  User.SkeletalMesh {
   type=NiagaraDataInterfaceSkeletalMesh
   defaultValue {
    struct=NiagaraExt_VariableValue_DataInterface
    value {dataInterface=&self:NiagaraDataInterfaceSkeletalMesh_0}
   }
  }
  User.Velocity {type=Vector3f, description="", defaultValue {struct=Vector3f, value=<0 0 0.2>}}
 ]
}
