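# Physics asset for the Twisted variant of the male medium Mastermind body, defining capsule bodies for dynamic goo bones on the left arm and dynamic bones at the right clavicle.
# Two two-link goo chains hang from their own parent bodies, which have no constraint to any other body.
# The clavicle chain runs from clavicle_r through a parent body and two links, then forks into A and B tips.
# All constraints leave swing and twist free.
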PhysicsAsset /BRCosmetics/Characters/Player/Male/Medium/Bodies/M_MED_Mastermind/Meshes/M_MED_Mastermind_Twisted_Physics
bodies {
 clavicle_r [{shapeType=Capsule, length=4.75, radius=6.5, center=<-8.0816 2.5528 0.7468>, rotation=<rot 90 0 0>}]
 dyn_clavA_01_r [{shapeType=Capsule, length=1.0308, radius=0.8058, center=<0.1964 0.0979 -0.1259>, rotation=<rot 0 -60.9453 -81.7266>}]
 dyn_clavB_01_r [{shapeType=Capsule, length=1.7103, radius=0.9092, center=<0.8581 0.0787 -0.0528>, rotation=<rot 0 -76.9352 -87.6492>}]
 dyn_clav_01_r [{shapeType=Capsule, length=1, radius=1, rotation=<rot 90 0 0>}]
 dyn_clav_02_r [{shapeType=Capsule, length=2, radius=1.505, rotation=<rot 90 0 0>}]
 dyn_goo_armB_01_l [{shapeType=Capsule, length=4.5, radius=1.755, center=<1.2622 0.0042 -0.0005>, rotation=<rot 89.0486 -179.7984 175.4681>}]
 dyn_goo_armB_02_l [{shapeType=Capsule, length=2.5, radius=1.7405, center=<0.805 -0.2393 0.2387>, rotation=<rot -0.0012 80.3202 -86.6224>}]
 dyn_goo_arm_01_l [{shapeType=Capsule, length=5.25, radius=2.505, center=<0.4182 0.0013 0>, rotation=<rot 90 -90 -90.2>}]
 dyn_goo_arm_02_l [{shapeType=Capsule, length=2.25, radius=1, center=<0.5496 0.0002 0>, rotation=<rot 90 0 0>}]
 dyn_parent_clav_r [{shapeType=Capsule, length=0.505, radius=0.505, rotation=<rot 0 -90 0>}]
 dyn_parent_goo_armB_l [{shapeType=Capsule, length=0.505, radius=0.505, rotation=<rot 0 -90 0>}]
 dyn_parent_goo_arm_l [{shapeType=Capsule, length=0.505, radius=0.505, rotation=<rot 0 -90 0>}]
}
constraints [
 {
  bone1Name=dyn_goo_arm_01_l
  bone2Name=dyn_parent_goo_arm_l
  swing1LimitDegrees=45
  swing1Motion=Free
  swing2LimitDegrees=45
  swing2Motion=Free
  twistLimitDegrees=45
  twistMotion=Free
 }
 {
  bone1Name=dyn_goo_armB_01_l
  bone2Name=dyn_parent_goo_armB_l
  swing1LimitDegrees=45
  swing1Motion=Free
  swing2LimitDegrees=45
  swing2Motion=Free
  twistLimitDegrees=45
  twistMotion=Free
 }
 {
  bone1Name=dyn_goo_armB_02_l
  bone2Name=dyn_goo_armB_01_l
  swing1LimitDegrees=45
  swing1Motion=Free
  swing2LimitDegrees=45
  swing2Motion=Free
  twistLimitDegrees=45
  twistMotion=Free
 }
 {
  bone1Name=dyn_goo_arm_02_l
  bone2Name=dyn_goo_arm_01_l
  swing1LimitDegrees=45
  swing1Motion=Free
  swing2LimitDegrees=45
  swing2Motion=Free
  twistLimitDegrees=45
  twistMotion=Free
 }
 {
  bone1Name=dyn_clav_01_r
  bone2Name=dyn_parent_clav_r
  swing1LimitDegrees=45
  swing1Motion=Free
  swing2LimitDegrees=45
  swing2Motion=Free
  twistLimitDegrees=45
  twistMotion=Free
 }
 {
  bone1Name=dyn_clav_02_r
  bone2Name=dyn_clav_01_r
  swing1LimitDegrees=45
  swing1Motion=Free
  swing2LimitDegrees=45
  swing2Motion=Free
  twistLimitDegrees=45
  twistMotion=Free
 }
 {
  bone1Name=dyn_parent_clav_r
  bone2Name=clavicle_r
  swing1LimitDegrees=45
  swing1Motion=Free
  swing2LimitDegrees=45
  swing2Motion=Free
  twistLimitDegrees=45
  twistMotion=Free
 }
 {
  bone1Name=dyn_clavB_01_r
  bone2Name=dyn_clav_02_r
  swing1LimitDegrees=45
  swing1Motion=Free
  swing2LimitDegrees=45
  swing2Motion=Free
  twistLimitDegrees=45
  twistMotion=Free
 }
 {
  bone1Name=dyn_clavA_01_r
  bone2Name=dyn_clav_02_r
  swing1LimitDegrees=45
  swing1Motion=Free
  swing2LimitDegrees=45
  swing2Motion=Free
  twistLimitDegrees=45
  twistMotion=Free
 }
]
