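# Physics asset for the dynamic jacket bones of the male medium Mastermind player body cosmetic.
# Thin boxes on front, side and back panel bones form short chains, constrained with limited swing and tight twist to small capsules on MSTR root bones.
# Extra constraints link the side panels to the front panels and the side-back panels to the center back's branches.
# Capsules on the pelvis, lower spine and thigh twist bones supply body collision shapes.
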PhysicsAsset /BRCosmetics/Characters/Player/Male/Medium/Bodies/M_MED_Mastermind/Meshes/M_MED_Mastermind_jacket_dyn
bodies {
 dyn_jacket_bk [{shapeType=Box, extentX=8, extentY=6, extentZ=2.8, center=<3 0 0>}]
 dyn_jacket_bk_MSTR [{shapeType=Capsule, length=1, radius=1, rotation=<rot 90 0 0>}]
 dyn_jacket_bk_l_01 [{shapeType=Box, extentX=6, extentY=5, extentZ=2.8, center=<3 0 0>}]
 dyn_jacket_bk_r_01 [{shapeType=Box, extentX=6, extentY=5, extentZ=2.8, center=<3 0 0>}]
 dyn_jacket_l_01 [{shapeType=Box, extentX=10, extentY=12, extentZ=2, center=<5 0 0>}]
 dyn_jacket_l_02 [{shapeType=Box, extentX=10, extentY=12, extentZ=2, center=<5 0 0>}]
 dyn_jacket_l_MSTR [{shapeType=Capsule, length=1, radius=1, rotation=<rot 90 0 0>}]
 dyn_jacket_l_bk_01 [{shapeType=Box, extentX=6, extentY=12, extentZ=2.8, center=<3 0 0>}]
 dyn_jacket_l_bk_02 [{shapeType=Box, extentX=6, extentY=12, extentZ=2.8, center=<4 0 0>}]
 dyn_jacket_l_bk_MSTR [{shapeType=Capsule, length=1, radius=1, rotation=<rot 90 0 0>}]
 dyn_jacket_l_fr_01 [{shapeType=Box, extentX=10, extentY=12, extentZ=2, center=<5 0.5 0>, rotation=<rot 0.0002 14.4632 0>}]
 dyn_jacket_l_fr_02 [{shapeType=Box, extentX=10, extentY=13, extentZ=2, center=<5.1254 2.508 0>, rotation=<rot 0.0006 14.5772 0>}]
 dyn_jacket_l_fr_MSTR [{shapeType=Capsule, length=1, radius=1, rotation=<rot 90 0 0>}]
 dyn_jacket_r_01 [{shapeType=Box, extentX=10, extentY=12, extentZ=2, center=<5 0 0>}]
 dyn_jacket_r_02 [{shapeType=Box, extentX=10, extentY=12, extentZ=2, center=<5 0 0>}]
 dyn_jacket_r_MSTR [{shapeType=Capsule, length=1, radius=1, rotation=<rot 90 0 0>}]
 dyn_jacket_r_bk_01 [{shapeType=Box, extentX=6, extentY=12, extentZ=2.8, center=<3 0 0>}]
 dyn_jacket_r_bk_02 [{shapeType=Box, extentX=6, extentY=12, extentZ=2.8, center=<4 0 0>}]
 dyn_jacket_r_bk_MSTR [{shapeType=Capsule, length=1, radius=1, rotation=<rot 90 0 0>}]
 dyn_jacket_r_fr_01 [{shapeType=Box, extentX=10, extentY=12, extentZ=2, center=<5 -0.5 0>, rotation=<rot 0.0003 -8.8025 0>}]
 dyn_jacket_r_fr_02 [{shapeType=Box, extentX=10, extentY=14, extentZ=2, center=<5.2528 -1.7158 0>, rotation=<rot 0.0007 -7.0863 0.0001>}]
 dyn_jacket_r_fr_MSTR [{shapeType=Capsule, length=1, radius=1, rotation=<rot 90 0 0>}]
 pelvis [
  {shapeType=Capsule, length=12.3239, radius=7.9771, center=<-2.7227 -1.1747 0>}
  {shapeType=Capsule, length=13.0936, radius=8.1762, center=<-7.049 -1.2079 0>}
 ]
 spine_01 [{shapeType=Capsule, length=10.2698, radius=7.3782, center=<-1.2151 -0.8323 0>}]
 thigh_twist_01_l [{shapeType=Capsule, length=22.5, radius=7.5, center=<11.485 -1 1.2712>, rotation=<rot 80.8083 179.967 174.2326>}]
 thigh_twist_01_r [{shapeType=Capsule, length=22.5, radius=7.5, center=<-11.485 1 -1.2712>, rotation=<rot 80.8083 179.967 -5.7674>}]
}
constraints [
 {bone1Name=dyn_jacket_bk, bone2Name=dyn_jacket_bk_MSTR, swing1LimitDegrees=5, swing2LimitDegrees=45, twistLimitDegrees=5}
 {bone1Name=dyn_jacket_bk_l_01, bone2Name=dyn_jacket_bk, swing1LimitDegrees=30, swing2LimitDegrees=30, twistLimitDegrees=5}
 {bone1Name=dyn_jacket_bk_r_01, bone2Name=dyn_jacket_bk, swing1LimitDegrees=30, swing2LimitDegrees=30, twistLimitDegrees=5}
 {bone1Name=dyn_jacket_r_bk_01, bone2Name=dyn_jacket_r_bk_MSTR, swing1LimitDegrees=45, swing2LimitDegrees=45, twistLimitDegrees=5}
 {bone1Name=dyn_jacket_r_bk_02, bone2Name=dyn_jacket_r_bk_01, swing1LimitDegrees=45, swing2LimitDegrees=45, twistLimitDegrees=5}
 {bone1Name=dyn_jacket_l_bk_01, bone2Name=dyn_jacket_l_bk_MSTR, swing1LimitDegrees=45, swing2LimitDegrees=45, twistLimitDegrees=5}
 {bone1Name=dyn_jacket_l_bk_02, bone2Name=dyn_jacket_l_bk_01, swing1LimitDegrees=45, swing2LimitDegrees=45, twistLimitDegrees=5}
 {bone1Name=dyn_jacket_l_fr_01, bone2Name=dyn_jacket_l_fr_MSTR, swing1LimitDegrees=30, swing2LimitDegrees=95, twistLimitDegrees=5}
 {bone1Name=dyn_jacket_l_fr_02, bone2Name=dyn_jacket_l_fr_01, swing1LimitDegrees=10, swing2LimitDegrees=75, twistLimitDegrees=5}
 {bone1Name=dyn_jacket_r_fr_01, bone2Name=dyn_jacket_r_fr_MSTR, swing1LimitDegrees=30, swing2LimitDegrees=95, twistLimitDegrees=5}
 {bone1Name=dyn_jacket_r_fr_02, bone2Name=dyn_jacket_r_fr_01, swing1LimitDegrees=10, swing2LimitDegrees=75, twistLimitDegrees=5}
 {bone1Name=dyn_jacket_l_02, bone2Name=dyn_jacket_l_01, swing1LimitDegrees=45, swing2LimitDegrees=45, twistLimitDegrees=5}
 {bone1Name=dyn_jacket_l_01, bone2Name=dyn_jacket_l_MSTR, swing1LimitDegrees=45, swing2LimitDegrees=45, twistLimitDegrees=5}
 {bone1Name=dyn_jacket_r_01, bone2Name=dyn_jacket_r_MSTR, swing1LimitDegrees=45, swing2LimitDegrees=45, twistLimitDegrees=5}
 {bone1Name=dyn_jacket_r_02, bone2Name=dyn_jacket_r_01, swing1LimitDegrees=45, swing2LimitDegrees=45, twistLimitDegrees=5}
 {bone1Name=dyn_jacket_l_02, bone2Name=dyn_jacket_l_fr_02, swing1LimitDegrees=45, swing2LimitDegrees=10, twistLimitDegrees=5}
 {bone1Name=dyn_jacket_r_02, bone2Name=dyn_jacket_r_fr_02, swing1LimitDegrees=45, swing2LimitDegrees=30, twistLimitDegrees=5}
 {bone1Name=dyn_jacket_r_bk_02, bone2Name=dyn_jacket_bk_r_01, swing1LimitDegrees=45, swing2LimitDegrees=30, twistLimitDegrees=5}
 {bone1Name=dyn_jacket_l_bk_02, bone2Name=dyn_jacket_bk_l_01, swing1LimitDegrees=45, swing2LimitDegrees=30, twistLimitDegrees=5}
]
