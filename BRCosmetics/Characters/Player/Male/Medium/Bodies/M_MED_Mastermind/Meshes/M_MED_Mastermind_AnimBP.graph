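# Animation Blueprint for the Mastermind medium male body cosmetic, driving secondary motion on a pose copied from the attached parent mesh.
# Weighted constraints position the right clavicle and left goo-arm helper bones, then cache the pose.
# With rigid bodies enabled, two chained simulations run: the jacket with skydiving force and gravity override, then the Twisted physics asset.
# Otherwise, thigh-up values drive additive jacket bone offsets; goo-arm Trail controllers are unconnected to the output.
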AnimBlueprint /BRCosmetics/Characters/Player/Male/Medium/Bodies/M_MED_Mastermind/Meshes/M_MED_Mastermind_AnimBP
target_skeleton=&/M_MED_Mastermind_Skeleton
graphs [
 AnimGraph {nodes [
   n1 = Trail controller - Bone: dyn_goo_arm_02_l(ComponentPose=Local To Component(), Alpha=1) {
    relaxationSpeedScale=3
    rotationLimits [{limitMax=<180 180 180>}, {limitMax=<180 180 180>}]
    rotationOffsets [<0 0 0>, <0 0 0>]
    trailBone {boneName=dyn_goo_arm_02_l}
    trailRelaxationSpeed {editorCurveData {keys [<at 0 10 RCIM_Linear>,<at 0.9789 7.2121 RCIM_Linear>]}}
    bLimitStretch=~
   }
   Component To Local(ComponentPose=@n1)
   n2 = RigidBody!(
    ComponentPose=Local To Component(LocalPose=Use cached pose 'PreCache'())
    Alpha
    OverrideWorldGravity:Vector=Get GravityOverride()
    ExternalForce=Break Fort Anim Input Skydiving External Force(FortAnimInput_SkydivingExternalForce=Get SkydivingExternalForceInput()) -> (
     bUseSkydivingVectorForce=~
     SkydivingVectorForce=~
     HeadToPelvisDirection=~
     FloatingMultiplier=~
     FloatingAdditive=~
     DivingMultiplier=~
     DivingAdditive=~
     ParachutingMultiplier=~
     ParachutingAdditive=~
    ).SkydivingVectorForce
   ) -> (Pose) {
    bClampLinearTranslationLimitToRefPose=true
    bOverrideWorldGravity=true
    overridePhysicsAsset=&/M_MED_Mastermind_jacket_dyn
    simulationSpace=BaseBoneSpace
    alphaBoolBlend {blendOption=Linear}
    baseBoneRef {boneName=spine_01}
    componentAppliedLinearAccClamp=<3000 3000 3000>
    componentLinearAccScale=<0.4 0.3 0.3>
    componentLinearVelScale=<0.4 0.3 0.3>
    overrideWorldGravity {z=-980}
   }
   n3 = RigidBody!(ComponentPose=@n2, Alpha, ExternalForce) -> (Pose) {
    bClampLinearTranslationLimitToRefPose=true
    bOverrideWorldGravity=true
    overridePhysicsAsset=&/M_MED_Mastermind_Twisted_Physics
    simulationSpace=BaseBoneSpace
    alphaBoolBlend {blendOption=Linear}
    baseBoneRef {boneName=root}
    componentAppliedLinearAccClamp=<1500 1500 1500>
    componentLinearAccScale=<0.4 0.3 0.3>
    componentLinearVelScale=<0.4 0.3 0.3>
    overrideWorldGravity {z=10}
   }
   n4 = Get ThighUpLeft() -> (ThighUpLeft:float)
   n5 = "Transform (Modify) Bone - Bone: dyn_jacket_l_fr_01"(ComponentPose=Local To Component(LocalPose=Use cached pose 'PreCache'()), Alpha=@n4, Translation="-2.7,0,0", Rotation="0,0,0", Scale="1,1,1") {
    rotationMode=BMM_Additive
    rotationSpace=BCS_BoneSpace
    translationMode=BMM_Additive
    translationSpace=BCS_BoneSpace
    alphaBoolBlend {blendOption=Linear}
    boneToModify {boneName=dyn_jacket_l_fr_01}
    translation {x=-2.7}
   }
   n6 = "Transform (Modify) Bone - Bone: dyn_jacket_l_fr_02"(ComponentPose=@n5, Alpha=@n4, Translation="-3.2,0,0", Rotation="95,-10,30.000", Scale="1,1,1") {
    rotationMode=BMM_Additive
    rotationSpace=BCS_BoneSpace
    translationMode=BMM_Additive
    translationSpace=BCS_BoneSpace
    alphaBoolBlend {blendOption=Linear}
    boneToModify {boneName=dyn_jacket_l_fr_02}
    rotation=<rot 95 -10 30>
    translation {x=-3.2}
   }
   n7 = "Transform (Modify) Bone - Bone: dyn_jacket_l_01"(ComponentPose=@n6, Alpha=@n4, Translation="-2.5,0,0", Rotation="15,15,-15", Scale="1,1,1") {
    rotationMode=BMM_Additive
    rotationSpace=BCS_BoneSpace
    translationMode=BMM_Additive
    translationSpace=BCS_BoneSpace
    alphaBoolBlend {blendOption=Linear}
    boneToModify {boneName=dyn_jacket_l_01}
    rotation=<rot 15 15 -15>
    translation {x=-2.5}
   }
   n8 = Get ThighUpRight() -> (ThighUpRight:float)
   n9 = "Transform (Modify) Bone - Bone: dyn_jacket_r_fr_01"(ComponentPose=@n7, Alpha=@n8, Translation="-2.7,0,0", Rotation="0,0,0", Scale="1,1,1") {
    rotationMode=BMM_Additive
    rotationSpace=BCS_BoneSpace
    translationMode=BMM_Additive
    translationSpace=BCS_BoneSpace
    alphaBoolBlend {blendOption=Linear}
    boneToModify {boneName=dyn_jacket_r_fr_01}
    translation {x=-2.7}
   }
   n10 = "Transform (Modify) Bone - Bone: dyn_jacket_r_fr_02"(ComponentPose=@n9, Alpha=@n8, Translation="-3.2,0,0", Rotation="95,10,-30", Scale="1,1,1") {
    rotationMode=BMM_Additive
    rotationSpace=BCS_BoneSpace
    translationMode=BMM_Additive
    translationSpace=BCS_BoneSpace
    alphaBoolBlend {blendOption=Linear}
    boneToModify {boneName=dyn_jacket_r_fr_02}
    rotation=<rot 95 10 -30>
    translation {x=-3.2}
   }
   n11 = "Transform (Modify) Bone - Bone: dyn_jacket_r_01"(ComponentPose=@n10, Alpha=@n8, Translation="-2.5,0,0", Rotation="15,-15,15", Scale="1,1,1") {
    rotationMode=BMM_Additive
    rotationSpace=BCS_BoneSpace
    translationMode=BMM_Additive
    translationSpace=BCS_BoneSpace
    alphaBoolBlend {blendOption=Linear}
    boneToModify {boneName=dyn_jacket_r_01}
    rotation=<rot 15 -15 15>
    translation {x=-2.5}
   }
   Output Pose(Result=Blend Poses by bool(BlendPose_0=Component To Local(ComponentPose=@n3), BlendPose_1=Component To Local(ComponentPose=@n11), bActiveValue=Get bRigidBodyEnabled()) {blendType=Linear}) {layerGroup=None}
   n12 = "Constraint - dyn_parent_clav_r (PPP)"!(
    ComponentPose=Local To Component(LocalPose=Copy Pose From Mesh() {bUseAttachedParent=true})
    Alpha=1
    ConstraintWeights_0=0.2351
    ConstraintWeights_1:float=0.5949
    ConstraintWeights_2:float=0.17
   ) -> (Pose) {
    alphaBoolBlend {blendOption=Linear}
    boneToModify {boneName=dyn_parent_clav_r}
    constraintSetup [{targetBone {boneName=upperarm_twist_01_r}}, {targetBone {boneName=clavicle_r}}, {targetBone {boneName=spine_05}}]
    constraintWeights [1,1,1]
   }
   n13 = "Constraint - dyn_parent_goo_arm_l (PPPPPPPP)"!(
    ComponentPose=@n12
    Alpha=1
    ConstraintWeights_0=0.0105
    ConstraintWeights_1:float=0.2247
    ConstraintWeights_2:float=0.0529
    ConstraintWeights_3:float=0.0029
    ConstraintWeights_4:float=0.6734
    ConstraintWeights_5:float=0.0004
    ConstraintWeights_6:float=0.0195
    ConstraintWeights_7:float=0.0158
   ) -> (Pose) {
    alphaBoolBlend {blendOption=Linear}
    boneToModify {boneName=dyn_parent_goo_arm_l}
    constraintSetup [
     {targetBone {boneName=head}}
     {targetBone {boneName=clavicle_l}}
     {targetBone {boneName=dyn_goo_arm_02_l}}
     {targetBone {boneName=spine_04}}
     {targetBone {boneName=spine_05}}
     {targetBone {boneName=upperarm_twist_01_l}}
     {targetBone {boneName=neck_02}}
     {targetBone {boneName=neck_01}}
    ]
    constraintWeights [1,1,1,1,1,1,1,1]
   }
   n14 = "Constraint - dyn_parent_goo_armB_l (PPPPPPP)"!(
    ComponentPose=@n13
    Alpha=1
    ConstraintWeights_0=0.0038
    ConstraintWeights_1:float=0.5198
    ConstraintWeights_2:float=0.0009
    ConstraintWeights_3:float=0.4528
    ConstraintWeights_4:float=0.0085
    ConstraintWeights_5:float=0.0072
    ConstraintWeights_6:float=0.007
   ) -> (Pose) {
    alphaBoolBlend {blendOption=Linear}
    boneToModify {boneName=dyn_parent_goo_armB_l}
    constraintSetup [
     {targetBone {boneName=head}}
     {targetBone {boneName=clavicle_l}}
     {targetBone {boneName=spine_04}}
     {targetBone {boneName=spine_05}}
     {targetBone {boneName=upperarm_twist_01_l}}
     {targetBone {boneName=neck_02}}
     {targetBone {boneName=neck_01}}
    ]
    constraintWeights [1,1,1,1,1,1,1]
   }
   Save cached pose 'PreCache'(Pose=Component To Local(ComponentPose=@n14)) {cacheName=PreCache}
   Trail controller - Bone: dyn_goo_arm_02_l(Alpha=1) {
    relaxationSpeedScale=3
    rotationLimits [{limitMax=<180 180 180>}, {limitMax=<180 180 180>}]
    rotationOffsets [<0 0 0>, <0 0 0>]
    trailBone {boneName=dyn_goo_arm_02_l}
    trailRelaxationSpeed {editorCurveData {keys [<at 0 10 RCIM_Linear>,<at 0.9789 7.2121 RCIM_Linear>]}}
    bLimitStretch=~
   }
  ]}
 EventGraph {links [], nodes []}
]
variables []
