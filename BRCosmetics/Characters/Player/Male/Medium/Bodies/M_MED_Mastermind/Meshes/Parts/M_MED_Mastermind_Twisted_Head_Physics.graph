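# Physics asset for the Twisted head part of the male medium Mastermind player body, covering its dynamic goo bones.
# Capsule bodies form three two-bone chains (gooA, gooB, gooC), each hanging from a small capsule on the head bone.
# Constraints link each root goo bone to the head and each tip to its root.
# All swing and twist motions are Free, so the stored angular limits do not restrict rotation.
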PhysicsAsset /BRCosmetics/Characters/Player/Male/Medium/Bodies/M_MED_Mastermind/Meshes/Parts/M_MED_Mastermind_Twisted_Head_Physics
bodies {
 dyn_head_gooA_01 [{shapeType=Capsule, length=2.5, radius=3.25, center=<0.3839 -0.3684 0.0001>, rotation=<rot 90 90 133.8163>}]
 dyn_head_gooA_02 [{shapeType=Capsule, length=3.4083, radius=3.0102, center=<0.0733 -0.2962 0.0728>, rotation=<rot 1.9623 41.7718 -87.1001>}]
 dyn_head_gooB_01 [{shapeType=Capsule, length=1.75, radius=2.25, center=<-0.0753 0 0>, rotation=<rot 90 0 0>}]
 dyn_head_gooB_02 [{shapeType=Capsule, length=1.5808, radius=1.7288, center=<1.2046 0.3788 -0.0092>, rotation=<rot 0 -43.3836 -80.5326>}]
 dyn_head_gooC_01 [{shapeType=Capsule, length=3.25, radius=2.25, center=<0.5758 0 0.1869>, rotation=<rot 72.0174 179.9999 -179.9999>}]
 dyn_head_gooC_02 [{shapeType=Capsule, length=3.75, radius=1, center=<0.3067 -0.4945 0>, rotation=<rot -90 -45 10.7305>}]
 head [{shapeType=Capsule, length=1, radius=1}]
}
constraints [
 {
  bone1Name=dyn_head_gooA_01
  bone2Name=head
  swing1LimitDegrees=45
  swing1Motion=Free
  swing2LimitDegrees=45
  swing2Motion=Free
  twistLimitDegrees=45
  twistMotion=Free
 }
 {
  bone1Name=dyn_head_gooB_01
  bone2Name=head
  swing1LimitDegrees=45
  swing1Motion=Free
  swing2LimitDegrees=45
  swing2Motion=Free
  twistLimitDegrees=45
  twistMotion=Free
 }
 {
  bone1Name=dyn_head_gooC_01
  bone2Name=head
  swing1LimitDegrees=45
  swing1Motion=Free
  swing2LimitDegrees=45
  swing2Motion=Free
  twistLimitDegrees=45
  twistMotion=Free
 }
 {
  bone1Name=dyn_head_gooB_02
  bone2Name=dyn_head_gooB_01
  swing1LimitDegrees=45
  swing1Motion=Free
  swing2LimitDegrees=45
  swing2Motion=Free
  twistLimitDegrees=45
  twistMotion=Free
 }
 {
  bone1Name=dyn_head_gooA_02
  bone2Name=dyn_head_gooA_01
  swing1LimitDegrees=45
  swing1Motion=Free
  swing2LimitDegrees=45
  swing2Motion=Free
  twistLimitDegrees=45
  twistMotion=Free
 }
 {
  bone1Name=dyn_head_gooC_02
  bone2Name=dyn_head_gooC_01
  swing1LimitDegrees=45
  swing1Motion=Free
  swing2LimitDegrees=45
  swing2Motion=Free
  twistLimitDegrees=45
  twistMotion=Free
 }
]
